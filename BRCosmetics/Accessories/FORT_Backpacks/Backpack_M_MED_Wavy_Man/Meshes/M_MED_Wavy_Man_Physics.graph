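# Physics asset for the Wavy Man back bling, a Battle Royale backpack cosmetic, defining its simplified collision shape.
# A single body on the backpack_root bone combines an angled capsule with an offset box to approximate the mesh.
# It has no constraints, so the asset describes one rigid piece with no jointed or dangling parts.
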PhysicsAsset /BRCosmetics/Accessories/FORT_Backpacks/Backpack_M_MED_Wavy_Man/Meshes/M_MED_Wavy_Man_Physics
bodies {backpack_root [
  {shapeType=Capsule, length=10.8213, radius=6.4668, center=<-3.6216 -5.2911 0.1456>, rotation=<rot 10.83 -79.6502 86.3523>}
  {shapeType=Box, extentX=42.3451, extentY=14.8391, extentZ=34.0245, center=<3.1778 7.0622 -4.2217>}
 ]}
constraints []
